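# Physics asset for the Cat Burglar backpack cosmetic, giving its handle and straps collision bodies and joint limits.
# Single capsules cover the pack, the handle and two segments per left and right strap.
# The handle joins the pack with both swings locked, leaving a limited twist hinge.
# Each strap chains from the pack through two joints with tight swing and wider twist, with the right side looser than the left.
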PhysicsAsset /BRCosmetics/Accessories/FORT_Backpacks/Backpack_F_MED_Cat_Burglar/Meshes/F_MED_Cat_Burglar_Pack_Physics
bodies {
 dyn_handle [{shapeType=Capsule, length=5.0422, radius=2.4189, center=<0.007 0.0052 1.7672>, rotation=<rot 0 -89.4726 -89.1341>}]
 dyn_straps_L_01 [{shapeType=Capsule, length=2.005, radius=1.755, center=<0.0209 -0.0273 -1.7975>, rotation=<rot 0 89.0584 2.5271>}]
 dyn_straps_L_02 [{shapeType=Capsule, length=1.9762, radius=1.8481, center=<0.1242 -0.0252 -2.0672>, rotation=<rot 0 -90.1352 -177.2347>}]
 dyn_straps_R_01 [{shapeType=Capsule, length=2.005, radius=1.755, center=<0.0717 0.0202 -1.7557>, rotation=<rot 0 -87.927 2.3596>}]
 dyn_straps_R_02 [{shapeType=Capsule, length=1.9761, radius=1.8481, center=<-0.0357 0.0224 -2.0251>, rotation=<rot 0 91.2648 -177.0681>}]
 pack [{shapeType=Capsule, length=23.2615, radius=8.5506, center=<0.3153 -5.9225 -1.875>, rotation=<rot 0 90.5634 -80.4321>}]
}
constraints [
 {
  bone1Name=dyn_handle
  bone2Name=pack
  swing1LimitDegrees=0
  swing1Motion=Locked
  swing2LimitDegrees=0
  swing2Motion=Locked
  twistLimitDegrees=20
 }
 {bone1Name=dyn_straps_L_02, bone2Name=dyn_straps_L_01, swing1LimitDegrees=5, swing2LimitDegrees=5, twistLimitDegrees=35}
 {bone1Name=dyn_straps_R_02, bone2Name=dyn_straps_R_01, swing1LimitDegrees=7, swing2LimitDegrees=7, twistLimitDegrees=55}
 {bone1Name=dyn_straps_L_01, bone2Name=pack, swing1LimitDegrees=5, swing2LimitDegrees=5, twistLimitDegrees=35}
 {bone1Name=dyn_straps_R_01, bone2Name=pack, swing1LimitDegrees=7, swing2LimitDegrees=7, twistLimitDegrees=55}
]
